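# Post-process AnimBP for the ScoffFrill quadruped cosmetic that moves its cape bones with the body.
# A PoseAsset blend feeds constraints tying front and mid cape bases to shoulders, thighs and spine.
# Each update, back-knee bend relative to the thigh sets alphas for additive back-cape offsets.
# Sprung Anim Dynamics then simulates eight cape bones, with right-side bones root-relative and only back01_r enabling its ankle sphere limit.
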AnimBlueprint /CosmeticCompanions/Assets/Quadruped/ScoffFrill/Meshes/ScoffFrill_Quadruped_Post_AnimBP
target_skeleton=&/ScoffFrill_Quadruped_Skeleton
graphs [
 AnimGraph {nodes [
   n1 = Input Pose()
   n2 = Get PoseAsset() -> (PoseAsset:Pose Asset Object Reference)
   n3 = "Constraint - base_cape_front_l (PP)"!(
    ComponentPose=Local To Component(LocalPose=Pose!(PoseAsset:Pose Asset Object Reference=@n2, SourcePose:Pose Link Structure=@n1) -> (Pose:Pose Link Structure) {is=title AnimGraphNode_PoseBlendNode, blendOption=Linear})
    Alpha=1
    ConstraintWeights_0=0.8
    ConstraintWeights_1:float=0.2
   ) -> (Pose) {
    boneToModify {boneName=base_cape_front_l}
    constraintSetup [{targetBone {boneName=shoulder_fr_l}}, {targetBone {boneName=spine_03}}]
    constraintWeights [1,1]
   }
   n4 = "Constraint - base_cape_front_r (PP)"!(ComponentPose=@n3, Alpha=1, ConstraintWeights_0=0.8, ConstraintWeights_1:float=0.2) -> (Pose) {
    boneToModify {boneName=base_cape_front_r}
    constraintSetup [{targetBone {boneName=shoulder_fr_r}}, {targetBone {boneName=spine_03}}]
    constraintWeights [1,1]
   }
   n5 = "Constraint - base_cape_mid_l (PP)"!(ComponentPose=@n4, Alpha=1, ConstraintWeights_0=0.7, ConstraintWeights_1:float=0.3) -> (Pose) {
    boneToModify {boneName=base_cape_mid_l}
    constraintSetup [{targetBone {boneName=thigh_bk_l}}, {targetBone {boneName=spine_01}}]
    constraintWeights [1,1]
   }
   n6 = "Constraint - base_cape_mid_r (PP)"!(ComponentPose=@n5, Alpha=1, ConstraintWeights_0=0.7, ConstraintWeights_1:float=0.3) -> (Pose) {
    boneToModify {boneName=base_cape_mid_r}
    constraintSetup [{targetBone {boneName=thigh_bk_r}}, {targetBone {boneName=spine_01}}]
    constraintWeights [1,1]
   }
   n7 = "Transform (Modify) Bone - Bone: base_cape_back01_l"(
    ComponentPose=@n6
    Alpha=Get knee_l_back() -> (knee_l_back:double)
    Translation="-4.439,-0.485,0.997"
    Rotation="1.545,5.043,17.361"
    Scale="1,1,1"
   ) {
    rotationMode=BMM_Additive
    rotationSpace=BCS_BoneSpace
    translationMode=BMM_Additive
    translationSpace=BCS_BoneSpace
    boneToModify {boneName=base_cape_back01_l}
    rotation=<rot 1.545 5.043 17.361>
    translation=<-4.439 -0.485 0.997>
   }
   n8 = "Transform (Modify) Bone - Bone: base_cape_back01_r"(
    ComponentPose=@n7
    Alpha=Get knee_r_back() -> (knee_r_back:double)
    Translation="-3.465,-3.682,-0.485"
    Rotation="2.382,25.359,-24.679"
    Scale="1,1,1"
   ) {
    rotationMode=BMM_Additive
    rotationSpace=BCS_BoneSpace
    translationMode=BMM_Additive
    translationSpace=BCS_BoneSpace
    boneToModify {boneName=base_cape_back01_r}
    rotation=<rot 2.3819 25.3588 -24.679>
    translation=<-3.465 -3.682 -0.485>
   }
   n9 = Anim Dynamics - Bone: dyn_cape_front_l(ComponentPose=@n8, Alpha=1, ExternalForce="0,0,0") {
    angularDampingOverride=0.9999
    angularSpringConstant=10000
    bAngularSpring=true
    bOverrideAngularDamping=true
    bUseGravityOverride=true
    gravityScale=0.35
    boundBone {boneName=dyn_cape_front_l}
    componentLinearAccScale=<1 1 1>
    componentLinearVelScale=<1 1 1>
    gravityOverride {z=-980}
    physicsBodyDefinitions [{
      boundBone {boneName=dyn_cape_front_l}
      boxExtents=<20 30 30>
      constraintSetup {angularLimitsMax=<0 0 5>, angularLimitsMin=<0 -30 -25>, angularTarget=<1 0 0>}
      localJointOffset=<4 0 0>
     }]
   }
   n10 = Anim Dynamics - Bone: dyn_cape_front_r(ComponentPose=@n9, Alpha=1, ExternalForce="0,0,0") {
    angularDampingOverride=0.9999
    angularSpringConstant=10000
    bAngularSpring=true
    bOverrideAngularDamping=true
    bUseGravityOverride=true
    gravityScale=0.35
    simulationSpace=RootRelative
    boundBone {boneName=dyn_cape_front_r}
    componentLinearAccScale=<1 1 1>
    componentLinearVelScale=<1 1 1>
    gravityOverride {z=-980}
    physicsBodyDefinitions [{
      boundBone {boneName=dyn_cape_front_r}
      boxExtents=<20 30 30>
      constraintSetup {angularLimitsMax=<0 30 5>, angularLimitsMin=<0 0 -25>, angularTarget=<1 0 0>}
      localJointOffset=<4 0 0>
     }]
   }
   n11 = Anim Dynamics - Bone: dyn_cape_mid_l(ComponentPose=@n10, Alpha=1, ExternalForce="0,0,0") {
    angularDampingOverride=0.9999
    angularSpringConstant=10000
    bAngularSpring=true
    bOverrideAngularDamping=true
    bUseGravityOverride=true
    gravityScale=0.35
    boundBone {boneName=dyn_cape_mid_l}
    componentLinearAccScale=<1 1 1>
    componentLinearVelScale=<1 1 1>
    gravityOverride {z=-980}
    physicsBodyDefinitions [{
      boundBone {boneName=dyn_cape_mid_l}
      boxExtents=<20 30 30>
      constraintSetup {angularLimitsMax=<0 0 30>, angularLimitsMin=<0 -30 0>, angularTarget=<1 0 0>}
      localJointOffset=<4 0 0>
     }]
   }
   n12 = Anim Dynamics - Bone: dyn_cape_mid_r(ComponentPose=@n11, Alpha=1, ExternalForce="0,0,0") {
    angularDampingOverride=0.9999
    angularSpringConstant=10000
    bAngularSpring=true
    bOverrideAngularDamping=true
    bUseGravityOverride=true
    gravityScale=0.35
    simulationSpace=RootRelative
    boundBone {boneName=dyn_cape_mid_r}
    componentLinearAccScale=<1 1 1>
    componentLinearVelScale=<1 1 1>
    gravityOverride {z=-980}
    physicsBodyDefinitions [{
      boundBone {boneName=dyn_cape_mid_r}
      boxExtents=<20 30 30>
      constraintSetup {angularLimitsMax=<0 30 30>, angularLimitsMin=<0 0 0>, angularTarget=<1 0 0>}
      localJointOffset=<4 0 0>
     }]
   }
   n13 = Anim Dynamics - Bone: dyn_cape_back01_l(ComponentPose=@n12, Alpha=1, ExternalForce="0,0,0") {
    angularDampingOverride=0.9999
    angularSpringConstant=10000
    bAngularSpring=true
    bOverrideAngularDamping=true
    bUseGravityOverride=true
    bUsePlanarLimit=false
    gravityScale=0.35
    boundBone {boneName=dyn_cape_back01_l}
    componentLinearAccScale=<1 1 1>
    componentLinearVelScale=<1 1 1>
    gravityOverride {z=-980}
    physicsBodyDefinitions [{
      boundBone {boneName=dyn_cape_back01_l}
      boxExtents=<20 30 30>
      constraintSetup {angularLimitsMax=<0 0 25>, angularLimitsMin=<0 -30 0>, angularTarget=<1 0 0>}
      localJointOffset=<4 0 0>
     }]
    sphericalLimits [{limitRadius=3, drivingBone {boneName=ankle_bk_l}, sphereLocalOffset=<0 5 0>}]
   }
   n14 = Anim Dynamics - Bone: dyn_cape_back02_l(ComponentPose=@n13, Alpha=1, ExternalForce="0,0,0") {
    angularDampingOverride=0.9999
    angularSpringConstant=10000
    bAngularSpring=true
    bOverrideAngularDamping=true
    bUseGravityOverride=true
    bUsePlanarLimit=false
    gravityScale=0.35
    boundBone {boneName=dyn_cape_back02_l}
    componentLinearAccScale=<1 1 1>
    componentLinearVelScale=<1 1 1>
    gravityOverride {z=-980}
    physicsBodyDefinitions [{
      boundBone {boneName=dyn_cape_back02_l}
      boxExtents=<20 30 30>
      constraintSetup {angularLimitsMax=<0 0 25>, angularLimitsMin=<0 -15 0>, angularTarget=<1 0 0>}
      localJointOffset=<4 0 0>
     }]
   }
   n15 = Anim Dynamics - Bone: dyn_cape_back01_r(ComponentPose=@n14, Alpha=1, ExternalForce="0,0,0") {
    angularDampingOverride=0.9999
    angularSpringConstant=10000
    bAngularSpring=true
    bOverrideAngularDamping=true
    bUseGravityOverride=true
    bUsePlanarLimit=false
    bUseSphericalLimits=true
    gravityScale=0.35
    simulationSpace=RootRelative
    boundBone {boneName=dyn_cape_back01_r}
    componentLinearAccScale=<1 1 1>
    componentLinearVelScale=<1 1 1>
    gravityOverride {z=-980}
    physicsBodyDefinitions [{
      boundBone {boneName=dyn_cape_back01_r}
      boxExtents=<20 30 30>
      constraintSetup {angularLimitsMax=<0 30 25>, angularLimitsMin=<0 0 0>, angularTarget=<1 0 0>}
      localJointOffset=<4 0 0>
     }]
    sphericalLimits [{limitRadius=3, drivingBone {boneName=ankle_bk_r}, sphereLocalOffset=<0 -5 0>}]
   }
   n16 = Anim Dynamics - Bone: dyn_cape_back02_r(ComponentPose=@n15, Alpha=1, ExternalForce="0,0,0") {
    angularDampingOverride=0.9999
    angularSpringConstant=10000
    bAngularSpring=true
    bOverrideAngularDamping=true
    bUseGravityOverride=true
    gravityScale=0.35
    simulationSpace=RootRelative
    boundBone {boneName=dyn_cape_back02_r}
    componentLinearAccScale=<1 1 1>
    componentLinearVelScale=<1 1 1>
    gravityOverride {z=-980}
    physicsBodyDefinitions [{
      boundBone {boneName=dyn_cape_back02_r}
      boxExtents=<20 30 30>
      constraintSetup {angularLimitsMax=<0 15 25>, angularLimitsMin=<0 0 0>, angularTarget=<1 0 0>}
      localJointOffset=<4 0 0>
     }]
   }
   Output Pose(Result=Component To Local(ComponentPose=@n16))
  ]}
 EventGraph {nodes [
   Event BlueprintUpdateAnimation()
   n1 = GetOwningComponent()
   > Set knee_l_back(knee_l_back:double=GetAlphaFromReferenceTransformDelta(SkeletalMeshComponent=@n1, BoneName=knee_bk_l, BaseName=thigh_bk_l, InputRange=(X=-30,Y=-90), TwistAxis="0,0,1")) -> (Output_Get:double)
   > Set knee_r_back(knee_r_back:double=GetAlphaFromReferenceTransformDelta(SkeletalMeshComponent=@n1, BoneName=knee_bk_r, BaseName=thigh_bk_r, InputRange=(X=-30,Y=-90), TwistAxis="0,0,1")) -> (Output_Get:double)
  ]}
]
variables [
 knee_l_back {type=number}
 knee_r_back {type=number}
 PoseAsset {type=PoseAsset, default=&/ScoffFrill_Quadruped_PoseAsset}
]
